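# Material function for the GroovyReader Play back bling that adds animated cloud and flicker effects to its Emissive Color and Base Color inputs.
# Square UV masks pick out regions: one gets layered, panning, parallax-offset cloud noise with distortion and a blue base tint, others set wall versus window flicker minimums.
# A sine wave maxed with random flicker modulates emissive, which is then scaled by time of day, and dims base color within two small masked squares.
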MaterialFunction /BRCosmetics/Accessories/FORT_Backpacks/Backpack_F_MED_GroovyReader_Play/Materials/MaterialFunctions/MF_GroovyReader_Play_MaterialEffects
graph {nodes [
  n1 = TextureCoordinate()
  n2 = Constant() {r=0.2}
  n3 = Constant() {r=0.12}
  n4 = Constant() {r=0.37}
  n5 = Constant() {r=0.43}
  n6 = Constant() {r=30}
  n7 = MaterialFunctionCall!(
   Texture Coordinate=TextureCoordinate()
   X Pos Right=Constant() {r=0.418}
   X Pos Left=Constant() {r=0.368}
   Y Pos Up=Constant() {r=0.497}
   Y Pos Down=Constant() {r=0.532}
   SquareSoftness=Constant() {r=10}
   RotateSquare
   RotationAmount
  ) {materialFunction=/Game/Accessories/FORT_Backpacks/Backpack_F_MED_GroovyReader_Play/Materials/MaterialFunctions/MF_SquareMaskGenerator}
  n8 = LinearInterpolate(
   A=ScalarParameter() {defaultValue=0.9, group=EmissiveRandomFlicker, parameterName=EmissiveFlickerMinValue_Walls}
   B=ScalarParameter() {defaultValue=0.5, group=EmissiveRandomFlicker, parameterName=EmissiveFlickerMinValue_Window}
   Alpha=Add(
    A=MaterialFunctionCall(Texture Coordinate=@n1, X Pos Right=@n2, X Pos Left=@n3, Y Pos Up=@n4, Y Pos Down=@n5, SquareSoftness=@n6) {materialFunction=&MF_SquareMaskGenerator}.SquareMask
    B=@n7.SquareMask
   )
  )
  n9 = TextureCoordinate()
  n10 = Multiply(A=@n9) {constB=1.7}
  n11 = Constant() {r=1}
  n12 = TextureSample(UVs=Panner(Coordinate=BumpOffset(Coordinate=@n10, Height=@n11)) {speedX=-0.006, speedY=0.002}) {texture=&AtmosphericCloudNoise01}
  n13 = BumpOffset(Coordinate=Multiply(A=@n9) {constB=2}, Height=Constant() {r=-1.15})
  n14 = TextureCoordinate()
  n15 = TextureSample(UVs=Panner(Coordinate=Multiply(A=@n14) {constB=3}, Time=Time()) {speedX=-0.1, speedY=0.06}) {texture=&AtmosphericCloudNoise01}
  n16 = TextureSample(UVs=Add(A=Panner(Coordinate=@n13) {speedX=-0.015, speedY=0.008}, B=Multiply(A=@n15.R) {constB=0.008})) {texture=&AtmosphericCloudNoise01}
  n17 = Multiply(A=Saturate(None=Add(A=@n12.R) {constB=0.3}), B=@n16.R)
  n18 = MaterialFunctionCall(
   Texture Coordinate=TextureCoordinate()
   X Pos Right=Constant() {r=0.422}
   X Pos Left=Constant() {r=0.188}
   Y Pos Up=Constant() {r=0.13}
   Y Pos Down=Constant() {r=0.32}
   SquareSoftness=Constant() {r=30}
  ) {materialFunction=&MF_SquareMaskGenerator}
  n19 = LinearInterpolate(B=Saturate(None=Add(A=@n17) {constB=0.2}), Alpha=@n18.SquareMask) {constA=1}
  n20 = Multiply(A=FunctionInput() {inputName=Emissive Color}, B=@n19)
  n21 = Sine(None=Multiply(A=ScalarParameter() {defaultValue=1, group=EmissiveRandomFlicker, parameterName=SineWaveMultiplierFrequency, sortPriority=36}, B=Time()))
  n22 = MaterialFunctionCall() {materialFunction=&/MF__FloatFlickerRandom}
  n23 = Max(A=Multiply(A=ConstantBiasScale(None=@n21)) {constB=0.885}, B=@n22.RandomFlickerFloat)
  n24 = Max(A=Multiply(A=@n8, B=@n20), B=Multiply(A=@n20, B=@n23))
  n25 = Constant() {r=1}
  n26 = Constant() {r=0.6}
  n27 = Constant() {r=0.5}
  FunctionOutput(None=Multiply(
    A=Multiply(A=@n24) {constB=1.15}
    B=MaterialFunctionCall(Value_Morning=@n25, Value_Day=@n25, Value_Evening=@n26, Value_Night=@n27) {materialFunction=&Scalar_Time_of_Day}.Result
   )) {outputName=Emissive Color}
  n28 = Saturate(None=Multiply(A=@n16.R) {constB=0.15})
  n29 = Constant3Vector() {constant=<rgba 0.0025 0.0628 0.2552 1>}
  n30 = Add(
   A=Multiply(A=FunctionInput() {inputName=Base Color, sortPriority=1, previewValue=<1 1 1 1>}, B=@n19)
   B=LinearInterpolate(B=Multiply(A=@n28, B=@n29), Alpha=@n18.SquareMask)
  )
  n31 = Multiply(A=Multiply(A=@n8, B=@n30)) {constB=0.1}
  n32 = Multiply(A=Saturate(None=Add(A=@n23) {constB=0.38}), B=@n30)
  n33 = MaterialFunctionCall(
   Texture Coordinate=TextureCoordinate()
   X Pos Right=Constant() {r=0.188}
   X Pos Left=Constant() {r=0.15}
   Y Pos Up=Constant() {r=0.399}
   Y Pos Down=Constant() {r=0.418}
   SquareSoftness=Constant() {r=10}
  ) {materialFunction=&MF_SquareMaskGenerator}
  FunctionOutput(None=LinearInterpolate(A=@n30, B=Max(A=@n31, B=@n32), Alpha=Add(A=@n33.SquareMask, B=@n7.SquareMask))) {outputName=Base Color, sortPriority=1}
 ]}
