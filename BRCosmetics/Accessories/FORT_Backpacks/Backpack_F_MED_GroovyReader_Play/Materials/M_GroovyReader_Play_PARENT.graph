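# Parent material for the GroovyReader Play back bling, built on the shared character material stack for skeletal meshes.
# Base character functions run PreFX, skin FX, PostFX and cinematic post FX on material attributes.
# A blue glow masked by the red channel of an additional mask texture is added to emissive.
# That emissive and base color pass through the GroovyReader-specific effects function, whose output is written back into the attributes.
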Material /BRCosmetics/Accessories/FORT_Backpacks/Backpack_F_MED_GroovyReader_Play/Materials/M_GroovyReader_Play_PARENT
graph {
 properties {
  bAutomaticallySetUsageInEditor=false
  bUseMaterialAttributes=true
  bUsedWithMeshDeformer=true
  bUsedWithMorphTargets=true
  bUsedWithSkeletalMesh=true
  pixelDepthOffsetMode=PDOM_Legacy
  refractionMethod=RM_IndexOfRefraction
  shadingModel=MSM_FromMaterialExpression
  subsurfaceProfile=&SS_HeroSkin_02
  usedShadingModels=Default Lit | Subsurface | Clear Coat | Subsurface Profile | Cloth | Eye
 }
 nodes [
  n1 = MaterialFunctionCall() {materialFunction=&BaseFN_CharacterFunction_PreFX}
  n2 = MaterialFunctionCall(
   AO=@n1.AO
   Distance Blend=@n1.AO
   Fresnel=@n1.AO
   Gmap=@n1.AO
   Pre Skinned Local Position=@n1.AO
   Material=MaterialFunctionCall(
    PreskinnedPosition=@n1.Pre skinned Local Position
    Gmap=@n1.Pre skinned Local Position
    MaterialAttributes=@n1.Pre skinned Local Position
    AO=@n1.Pre skinned Local Position
   ) {materialFunction=&CharacterSkinFX}.Result
  ) {materialFunction=&BaseFN_CharacterFunction_PostFX}
  n3 = GetMaterialAttributes(None=MaterialFunctionCall(In=@n2.Result) {materialFunction=&CINE_MF_CharPostFX}.Result) {attributeGetTypes [#1, #2]}
  n4 = TextureSample() {
   samplerType=SAMPLERTYPE_Masks
   texture=/BRCosmetics/Accessories/FORT_Backpacks/Backpack_F_MED_GroovyReader_Play/Textures/T_GroovyReader_Play_Backpack_AdditionalMask
  }
  n5 = Multiply(A=@n4.R, B=Constant3Vector() {constant=<rgba 0 0.1356 0.3968 1>})
  n6 = MaterialFunctionCall!(Emissive Color=Add(A=@n3.Emissive Color, B=Multiply(A=@n5) {constB=0.22}), Base Color=@n3.Base Color) {materialFunction=/BRCosmetics/Accessories/FORT_Backpacks/Backpack_F_MED_GroovyReader_Play/Materials/MaterialFunctions/MF_GroovyReader_Play_MaterialEffects}
  SetMaterialAttributes!(MaterialAttributes=@n3.MaterialAttributes, Emissive Color=@n6.Emissive Color, Base Color=@n6.Emissive Color) {attributeSetTypes [#1, #2]}
 ]
}
